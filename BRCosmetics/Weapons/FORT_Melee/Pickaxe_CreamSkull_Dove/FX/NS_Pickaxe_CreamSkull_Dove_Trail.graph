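# Niagara ribbon trail effect for the CreamSkull Dove pickaxe, exposing user inputs for swing state and ribbon ID offset.
# Trail_Body and Trail_Camera spawn short-lived anim trails, with Body using explicit bone/socket indices and Camera setting its own width.
# Both fade alpha by a cubic curve scaled by the system's swing fade and time-of-day alpha, doubled for Camera.
# A system time-of-day color multiplier tints both ribbons.
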NiagaraSystem /BRCosmetics/Weapons/FORT_Melee/Pickaxe_CreamSkull_Dove/FX/NS_Pickaxe_CreamSkull_Dove_Trail
dependencies {
 usedDataInterfaces [&NiagaraDataInterfaceCurve, &NiagaraDataInterfaceSocketReader, &NiagaraDataInterfaceColorCurve]
 usedDynamicInputs [&ColorFromCurve, &FNTimeOfDayScalar]
 usedModules [
  &self:SystemSpawnScript.SystemScriptSource.SystemScriptGraph.NiagaraNodeAssignment_3.SetVariables_#1
  &SystemState
  &self:SystemSpawnScript.SystemScriptSource.SystemScriptGraph.NiagaraNodeAssignment_1.SetVariables_#2
  &self:SystemSpawnScript.SystemScriptSource.SystemScriptGraph.NiagaraNodeAssignment_0.SetVariables_#3
  &self:SystemSpawnScript.SystemScriptSource.SystemScriptGraph.NiagaraNodeAssignment_4.SetVariables_#4
 ]
}
emitters {
 Trail_Body {
  emitterSpawnScript [&FN_RibbonIDOffsetUtility()]
  emitterUpdateScript [
   &EmitterState()
   &FN_SpawnAnimTrail()
  ]
  particleSpawnScript [
   &FN_InitializeAnimTrailBoneIndecies(First Bone or Socket Index=2, Second Bone or Socket Index=3)
   &FN_InitializeAnimTrail(Lifetime=0.18)
  ]
  particleUpdateScript [
   &FN_UpdateAnimTrail()
   &ScaleColor(
    Scale Alpha=&Multiply_Float(
     A=&FloatFromCurve(
      FloatCurve=DataInterface() {propertyValues {Curve {keys [{interpMode=RCIM_Cubic, leaveTangent=-4.0138, tangentMode=RCTM_Break, time=0, value=1}, <at 0.4 0>]}}}
      Scale Curve=$System.SwingGeneralFade
     )
     B=$System.TOD_Alpha
    )
    Scale RGB=&MakeVectorFromLinearColorRGB(Color=$System.TODColorMultiplier)
    ScaleA=true
    ScaleRGB=true
   )
  ]
  rendererClasses [&NiagaraRibbonRendererProperties]
  renderers [{rendererClass=&NiagaraRibbonRendererProperties}]
 }
 Trail_Camera {
  emitterSpawnScript [&FN_RibbonIDOffsetUtility()]
  emitterUpdateScript [
   &EmitterState()
   &FN_SpawnAnimTrail()
  ]
  particleSpawnScript [
   &FN_InitializeAnimTrailBoneIndecies()
   &FN_InitializeAnimTrail(AnimTrailWidth=1.7, Lifetime=0.18)
  ]
  particleUpdateScript [
   &FN_UpdateAnimTrail()
   &ScaleColor(
    Scale Alpha=&Multiply_Float(
     A=&FloatFromCurve(
      FloatCurve=DataInterface() {propertyValues {Curve {keys [{interpMode=RCIM_Cubic, leaveTangent=-4.0138, tangentMode=RCTM_Break, time=0, value=1}, <at 0.4 0>]}}}
      Scale Curve=$System.SwingGeneralFade
     )
     B=&Multiply_Float(A=2, B=$System.TOD_Alpha)
    )
    Scale RGB=&MakeVectorFromLinearColorRGB(Color=$System.TODColorMultiplier)
    ScaleA=true
    ScaleRGB=true
   )
  ]
  rendererClasses [&NiagaraRibbonRendererProperties]
  renderers [{rendererClass=&NiagaraRibbonRendererProperties}]
 }
}
summary {
 systemName=NS_Pickaxe_CreamSkull_Dove_Trail
 emitters [
  {emitterName=Trail_Body, rendererClasses [&NiagaraRibbonRendererProperties]}
  {emitterName=Trail_Camera, rendererClasses [&NiagaraRibbonRendererProperties]}
 ]
 userVariables [
  User.bisSwinging {type=NiagaraBool}
  User.RibbonIDOffset {type=NiagaraInt32}
 ]
}
